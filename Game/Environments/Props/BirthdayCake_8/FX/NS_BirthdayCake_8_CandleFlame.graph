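# Niagara system for the candle flames on the BirthdayCake_8 prop, with blue-tinted flames, rising glow particles, warm flickering lights and smoke ribbons.
# Paired emitters sample static mesh sockets, one set for eleven numbered candles and another for the top candle.
# Waveforms and curves drive flicker, and flame color dims in the front end.
# A lagged-position delta feeds dynamic material parameters so flames respond to movement.
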NiagaraSystem /Game/Environments/Props/BirthdayCake_8/FX/NS_BirthdayCake_8_CandleFlame
dependencies {
 usedDataInterfaces [&NiagaraDataInterfaceStaticMesh, &NiagaraDataInterfaceCurve]
 usedDynamicInputs []
 usedModules [&SystemState]
}
emitters {
 FlamesAndGlows {
  emitterUpdateScript [
   &EmitterState(
    Inactive Response=::NewEnumerator1
    Life Cycle Mode=::NewEnumerator1
    Loop Behavior=::NewEnumerator1
    Loop Duration Mode=::NewEnumerator1
    UseLoopDelay=false
   )
   &SpawnBurst_Instantaneous(Spawn Count=11, Use Spawn Probability=false, UseLoopCountLimit=false)
  ]
  particleSpawnScript [
   &FN_SetSpawnGroup(Number of Spawn Groups=11)
   &InitializeParticle(
    Color=<rgba 0 2.6667 8 3>
    Lifetime Max=0.5
    Lifetime Min=0.1
    Lifetime Mode=::NewEnumerator1
    Mesh Scale=&MakeVector(X=&RandomRangeFloat(Minimum=2.4, Maximum=3.2), Y=2, Z=2)
    Mesh Scale Mode=::NewEnumerator4
    Sprite Size Mode=::NewEnumerator1
    Uniform Sprite Size Max=22
    Uniform Sprite Size Min=18
    UsePositionOffset=true
    Write Material Random=true
   )
   &self:SampleStaticMeshSockets(
    SM=DataInterface() {propertyValues {
      DefaultMesh=null
      InstanceIndex=-1
      LODIndex=0
      PreviewMesh=/Game/Environments/Props/BirthdayCake_5/Meshes/FN_Birthday_Tiered_Cake_5
      SoftSourceActor=null
      SourceMode=Default
      bAllowSamplingFromStreamingLODs=false
      bCaptureTransformsPerFrame=true
      bUsePhysicsBodyVelocity=false
      FilteredSockets [
       Candle_FXSocket_01
       Candle_FXSocket_02
       Candle_FXSocket_03
       Candle_FXSocket_04
       Candle_FXSocket_05
       Candle_FXSocket_06
       Candle_FXSocket_07
       Candle_FXSocket_08
       Candle_FXSocket_09
       Candle_FXSocket_10
       Candle_FXSocket_11
      ]
      LODIndexUserParameter {parameter {name=None, typeDefHandle {registeredTypeIndex=108}, varData []}}
      MeshParameterBinding {parameter {name=None, typeDefHandle {registeredTypeIndex=1}, varData []}}
      SectionFilter {allowedMaterialSlots []}
     }}
    Socket:NiagaraInt32=$Particles.SpawnGroup
   ) {module=SampleStaticMeshSockets}
   /SetVariables(Particles.ColorAlt:LinearColor=<rgba 0 0.5 1 0.2>) {isSetParameters=true, module=SetVariables_#1}
  ]
  particleUpdateScript [
   &ParticleState(Kill Particles When Lifetime Has Elapsed=false, Let Infinitely Lived Particles Die When Emitter Deactivates=true)
   &ScaleMeshSize(Scale Factor=&MakeVector(
     X=&Waveform(
      Global Input Position=&RandomRangeFloat(Minimum=0, Maximum=1)
      "Waveform [1]"=::NewEnumerator0
      "[1] Frequency"=1
      "[1] Amplitude Min/Max"=<0.7 1>
     )
     Y=1
     Z=1
    ))
   &SolveForcesAndVelocity(Clamp Velocity=false, Limit Acceleration=false)
   &ScaleColor(
    Scale Alpha=&Waveform(
     Global Input Position=&RandomRangeFloat(Minimum=0, Maximum=1)
     Global Phase=$System.NormalizedLoopAge
     "Waveform [1]"=::NewEnumerator12
     "[1] Frequency"=1
     "[1] Amplitude Min/Max"=<0.5 1>
    )
    Scale RGB=&VectorFromFloat(Value=&Waveform(
      Global Input Position=&RandomRangeFloat(Minimum=0, Maximum=1)
      Global Phase=$Particles.NormalizedAge
      "Waveform [1]"=::NewEnumerator12
      "[1] Frequency"=2
      "[1] Amplitude Min/Max"=<1 2>
     ))
    ScaleA=true
    ScaleRGB=true
   )
   &ScaleSpriteSize(Scale Factor=&Multiply_Vector2D_ByFloat(
     Float=&Waveform("Waveform [1]"=::NewEnumerator12, "[1] Frequency"=5, "[1] Amplitude Min/Max"=<1 2>, "[1] Scale"=0.5, "[1] Bias"=0.5)
     Vector2D=<1 1>
    ))
   /SetVariables(
    Particles.ColorAlt:LinearColor=&MultiplyLinearColor(
     Linear Color=$Particles.Initial.ColorAlt
     Scale Factor=&MakeVector4_FromFloat(Float=&Waveform(
       "Waveform [1]"=::NewEnumerator12
       "[1] Input Position"=&ReturnNormalizedExecIndex(Normalized Index Scale=0.8)
       "[1] Frequency"=3
       "[1] Amplitude Min/Max"=<0.5 1>
      ))
    )
    Particles.SpriteSizeAlt:Vector2f=&Vector2DFromFloat(Value=&FloatFromCurve(
      FloatCurve=DataInterface() {propertyValues {Curve {keys [
          {arriveTangent=-0.0036, interpMode=RCIM_Cubic, leaveTangent=-0.0036, tangentMode=RCTM_User, time=0, value=1}
          {arriveTangent=-2.2201, interpMode=RCIM_Cubic, leaveTangent=-2.2201, tangentMode=RCTM_User, time=1, value=0}
         ]}}}
      Scale Curve=&RandomRangeFloat(Minimum=5, Maximum=10)
     ))
   ) {isSetParameters=true, module=SetVariables_#2}
   /SetVariables(Particles.MaterialRandom:NiagaraFloat=&RandomRangeFloat(Minimum=0, Maximum=1)) {isSetParameters=true, module=SetVariables_#3}
   /SetVariables(Particles.LaggedPosition:Vector3f=/Niagara/DynamicInputs/Lerp/SmoothLerpOverTime_Position(
     Convergence Rate:NiagaraFloat=18
     Smooth Value:NiagaraPosition=Linked() {linkedVariable {name=Particles.LaggedPosition, type=Vector3f}}
     Target Value:NiagaraPosition=$Particles.Position
    )) {isSetParameters=true, module=SetVariables_#4}
   /SetVariables(Particles.PositionDelta:Vector3f=&Subtract_Vector(A=$Particles.LaggedPosition, B=Linked() {linkedVariable {name=Particles.Position, type=NiagaraPosition}})) {isSetParameters=true, module=SetVariables_#5}
   &ScaleColor(
    Scale RGB=&VectorFromFloat(Value=&FN_FrontEndFloatModifier(Gameplay=1, Front End=0.15))
    ScaleA=false
    ScaleRGB=true
   ) {module=ScaleColor001}
   &DynamicMaterialParameters(
    Index 0 Param 1=&Lerp_Float(
     A=1
     B=2
     Alpha=&Waveform(
      Global Input Position=&RandomRangeFloat(Minimum=0, Maximum=1)
      "Waveform [1]"=::NewEnumerator12
      "[1] Frequency"=5
      "[1] Amplitude Min/Max"=<0 1>
     )
    )
    Index 0 Param 2=&MakeFloatFromVector(Vector=$Particles.PositionDelta, Channel=::NewEnumerator0)
    Index 0 Param 3=&MakeFloatFromVector(Vector=$Particles.PositionDelta, Channel=::NewEnumerator1)
    Index 0 Param 4=&ClampFloat(Float=&MakeFloatFromVector(Vector=$Particles.PositionDelta, Channel=::NewEnumerator2), Max=2)
    Module.Index0Param0Write=true
    Module.Index0Param1Write=true
    Module.Index0Param2Write=true
    Module.Index0Param3Write=true
   )
   &CameraOffset(Camera Offset Amount=2)
  ]
  rendererClasses [&NiagaraSpriteRendererProperties, &NiagaraMeshRendererProperties]
  renderers [
   {rendererClass=&NiagaraSpriteRendererProperties}
   {rendererClass=&NiagaraMeshRendererProperties, rendererIndex=1}
  ]
 }
 FlamesAndGlows001 {
  emitterUpdateScript [
   &EmitterState(
    Inactive Response=::NewEnumerator1
    Life Cycle Mode=::NewEnumerator1
    Loop Behavior=::NewEnumerator1
    Loop Duration Mode=::NewEnumerator1
    UseLoopDelay=false
   )
   &SpawnBurst_Instantaneous(Use Spawn Probability=false, UseLoopCountLimit=false)
  ]
  particleSpawnScript [
   &FN_SetSpawnGroup(Number of Spawn Groups=1)
   &InitializeParticle(
    Color=<rgba 0 3.3333 5 3>
    Lifetime Max=0.5
    Lifetime Min=0.1
    Lifetime Mode=::NewEnumerator1
    Mesh Scale=&MakeVector(X=&RandomRangeFloat(Minimum=3.6, Maximum=4), Y=2, Z=2)
    Mesh Scale Mode=::NewEnumerator4
    Sprite Size Mode=::NewEnumerator1
    Uniform Sprite Size Max=44
    Uniform Sprite Size Min=36
    UsePositionOffset=true
    Write Material Random=true
   )
   &self:SampleStaticMeshSockets(
    SM=DataInterface() {propertyValues {
      DefaultMesh=null
      InstanceIndex=-1
      LODIndex=0
      PreviewMesh=/Game/Environments/Props/BirthdayCake_6/Meshes/FN_Birthday_TieredCake_6
      SoftSourceActor=null
      SourceMode=Default
      bAllowSamplingFromStreamingLODs=false
      bCaptureTransformsPerFrame=true
      bUsePhysicsBodyVelocity=false
      FilteredSockets [Candle_FXSocket_Top]
      LODIndexUserParameter {parameter {name=None, typeDefHandle {registeredTypeIndex=108}, varData []}}
      MeshParameterBinding {parameter {name=None, typeDefHandle {registeredTypeIndex=1}, varData []}}
      SectionFilter {allowedMaterialSlots []}
     }}
    Socket:NiagaraInt32=$Particles.SpawnGroup
   ) {module=SampleStaticMeshSockets001}
   /SetVariables(Particles.ColorAlt:LinearColor=<rgba 0 0.5 1 0.3>) {isSetParameters=true, module=SetVariables_#1}
  ]
  particleUpdateScript [
   &ParticleState(Kill Particles When Lifetime Has Elapsed=false, Let Infinitely Lived Particles Die When Emitter Deactivates=true)
   &ScaleMeshSize(Scale Factor=&MakeVector(
     X=&Waveform(
      Global Input Position=&RandomRangeFloat(Minimum=0, Maximum=1)
      "Waveform [1]"=::NewEnumerator0
      "[1] Frequency"=1
      "[1] Amplitude Min/Max"=<0.7 1>
     )
     Y=1
     Z=1
    ))
   &SolveForcesAndVelocity(Clamp Velocity=false, Limit Acceleration=false)
   &ScaleColor(
    Scale Alpha=&Waveform(
     Global Input Position=&RandomRangeFloat(Minimum=0, Maximum=1)
     Global Phase=$System.NormalizedLoopAge
     "Waveform [1]"=::NewEnumerator12
     "[1] Frequency"=1
     "[1] Amplitude Min/Max"=<0.5 1>
    )
    Scale RGB=&VectorFromFloat(Value=&Waveform(
      Global Input Position=&RandomRangeFloat(Minimum=0, Maximum=1)
      Global Phase=$Particles.NormalizedAge
      "Waveform [1]"=::NewEnumerator12
      "[1] Frequency"=2
      "[1] Amplitude Min/Max"=<1 2>
     ))
    ScaleA=true
    ScaleRGB=true
   )
   &ScaleSpriteSize(Scale Factor=&Multiply_Vector2D_ByFloat(
     Float=&Waveform("Waveform [1]"=::NewEnumerator12, "[1] Frequency"=5, "[1] Amplitude Min/Max"=<1 2>, "[1] Scale"=0.5, "[1] Bias"=0.5)
     Vector2D=<1 1>
    ))
   /SetVariables(
    Particles.ColorAlt:LinearColor=&MultiplyLinearColor(
     Linear Color=$Particles.Initial.ColorAlt
     Scale Factor=&MakeVector4_FromFloat(Float=&Waveform(
       "Waveform [1]"=::NewEnumerator12
       "[1] Input Position"=&ReturnNormalizedExecIndex(Normalized Index Scale=0.8)
       "[1] Frequency"=3
       "[1] Amplitude Min/Max"=<0.5 1>
      ))
    )
    Particles.SpriteSizeAlt:Vector2f=&Vector2DFromFloat(Value=&FloatFromCurve(
      FloatCurve=DataInterface() {propertyValues {Curve {keys [
          {arriveTangent=-0.0036, interpMode=RCIM_Cubic, leaveTangent=-0.0036, tangentMode=RCTM_User, time=0, value=1}
          {arriveTangent=-2.2201, interpMode=RCIM_Cubic, leaveTangent=-2.2201, tangentMode=RCTM_User, time=1, value=0}
         ]}}}
      Scale Curve=&RandomRangeFloat(Minimum=5, Maximum=10)
     ))
   ) {isSetParameters=true, module=SetVariables_#2}
   /SetVariables(Particles.MaterialRandom:NiagaraFloat=&RandomRangeFloat(Minimum=0, Maximum=1)) {isSetParameters=true, module=SetVariables_#3}
   /SetVariables(Particles.LaggedPosition:Vector3f=&SmoothLerpOverTime_Position(
     Convergence Rate:NiagaraFloat=18
     Smooth Value:NiagaraPosition=Linked() {linkedVariable {name=Particles.LaggedPosition, type=Vector3f}}
     Target Value:NiagaraPosition=$Particles.Position
    )) {isSetParameters=true, module=SetVariables_#4}
   /SetVariables(Particles.PositionDelta:Vector3f=&Subtract_Vector(A=$Particles.LaggedPosition, B=Linked() {linkedVariable {name=Particles.Position, type=NiagaraPosition}})) {isSetParameters=true, module=SetVariables_#5}
   &ScaleColor(
    Scale RGB=&VectorFromFloat(Value=&FN_FrontEndFloatModifier(Gameplay=1, Front End=0.15))
    ScaleA=false
    ScaleRGB=true
   ) {module=ScaleColor001}
   &DynamicMaterialParameters(
    Index 0 Param 1=&Lerp_Float(
     A=1
     B=2
     Alpha=&Waveform(
      Global Input Position=&RandomRangeFloat(Minimum=0, Maximum=1)
      "Waveform [1]"=::NewEnumerator12
      "[1] Frequency"=5
      "[1] Amplitude Min/Max"=<0 1>
     )
    )
    Index 0 Param 2=&MakeFloatFromVector(Vector=$Particles.PositionDelta, Channel=::NewEnumerator0)
    Index 0 Param 3=&MakeFloatFromVector(Vector=$Particles.PositionDelta, Channel=::NewEnumerator1)
    Index 0 Param 4=&ClampFloat(Float=&MakeFloatFromVector(Vector=$Particles.PositionDelta, Channel=::NewEnumerator2), Max=2)
    Module.Index0Param0Write=true
    Module.Index0Param1Write=true
    Module.Index0Param2Write=true
    Module.Index0Param3Write=true
   )
   &CameraOffset(Camera Offset Amount=2)
  ]
  rendererClasses [&NiagaraSpriteRendererProperties, &NiagaraMeshRendererProperties]
  renderers [
   {rendererClass=&NiagaraSpriteRendererProperties}
   {rendererClass=&NiagaraMeshRendererProperties, rendererIndex=1}
  ]
 }
 Glow {
  emitterUpdateScript [
   &EmitterState()
   &SpawnRate(SpawnRate=150, Use Spawn Probability=false)
  ]
  particleSpawnScript [
   &FN_SetSpawnGroup(Normalize Spawn Group=true, Number of Spawn Groups=11, Particles Per Group=10, Set Ribbon ID and Ribbon Link Order=true)
   &InitializeParticle(
    Color=<rgba 0 3.3333 10 1>
    Lifetime=&Multiply_Float(A=&RandomRangeFloat(Minimum=0.2, Maximum=0.5), B=1)
    Sprite Size Mode=::NewEnumerator3
    Uniform Sprite Size=&Multiply_Float(A=&RandomRangeFloat(Minimum=0.65, Maximum=1), B=3)
    UsePositionOffset=false
    Write Material Random=true
   ) {module=InitializeParticle001}
   /Niagara/Modules/Spawn/Initialization/InitializeParticle(
    Color=<rgba 15 10 1 1>
    Lifetime=&Multiply_Float(A=&UniformRangedFloat(Minimum=0.2, Maximum=0.5), B=1)
    Module.Write Color=true
    Module.Write Lifetime=true
    Module.Write Mass=true
    Module.Write Material Random=true
    Module.Write Position=true
    Module.Write Scale=false
    Module.Write SpriteRotation=false
    Module.Write SpriteSize=true
    Position=&SimulationPosition()
    Sprite Size=&Vector2DFromFloat(Value=&Multiply_Float(A=&UniformRangedFloat(Minimum=0.65, Maximum=1), B=3))
   ) {enabled=false}
   &self:SampleStaticMeshSockets(
    SM=DataInterface() {propertyValues {
      DefaultMesh=null
      InstanceIndex=-1
      LODIndex=0
      PreviewMesh=&FN_Birthday_TieredCake_6
      SoftSourceActor=null
      SourceMode=Default
      bAllowSamplingFromStreamingLODs=false
      bCaptureTransformsPerFrame=true
      bUsePhysicsBodyVelocity=false
      FilteredSockets [
       Candle_FXSocket_01
       Candle_FXSocket_02
       Candle_FXSocket_03
       Candle_FXSocket_04
       Candle_FXSocket_05
       Candle_FXSocket_06
       Candle_FXSocket_07
       Candle_FXSocket_08
       Candle_FXSocket_09
       Candle_FXSocket_10
       Candle_FXSocket_11
      ]
      LODIndexUserParameter {parameter {name=None, typeDefHandle {registeredTypeIndex=108}, varData []}}
      MeshParameterBinding {parameter {name=None, typeDefHandle {registeredTypeIndex=1}, varData []}}
      SectionFilter {allowedMaterialSlots []}
     }}
    Socket:NiagaraInt32=$Particles.SpawnGroup
   ) {module=SampleStaticMeshSockets001}
   &Color(Color=&MultiplyLinearColor(
     Linear Color=$Particles.Color
     Scale Factor=/Niagara/DynamicInputs/UniformRange/UniformRangedVector4(
      Minimum:Vector4f=<1 0.25 0.25 0.01>
      Maximum:Vector4f=<1 1 1 1>
      Randomness Mode:ENiagaraRandomnessMode=::NewEnumerator0
      Random Seed:NiagaraInt32=0
      Fixed Random Seed:NiagaraBool=false
     )
    ))
  ]
  particleUpdateScript [
   &ParticleState()
   &AddVelocity(Coordinate Space=::World, Velocity=&RandomRangeVector(Minimum=<-0.25 -0.25 2>, Maximum=<0.25 0.25 10>))
   &Drag(Drag=25, Use Linear Drag=true, Use Rotational Drag=false)
   &SolveForcesAndVelocity(Clamp Velocity=false, Limit Acceleration=false)
   &ScaleColor(
    Scale Alpha=&FloatFromCurve(FloatCurve=DataInterface() {propertyValues {Curve {keys [<at 0.5 1>,<at 1 0>]}}})
    Scale RGB=&Lerp_Vector(A=<1 1 1>, B=<1 0 0>, Alpha=$Particles.NormalizedAge)
    ScaleA=true
    ScaleRGB=true
   )
   &ScaleSpriteSize(Scale Factor=&Vector2DFromFloat(Value=&FloatFromCurve(
      FloatCurve=DataInterface() {propertyValues {Curve {
         defaultValue=0
         keys [
          {arriveTangent=4.1785, interpMode=RCIM_Cubic, leaveTangent=4.1785, tangentMode=RCTM_User, time=0, value=0}
          {arriveTangent=1.1179, interpMode=RCIM_Cubic, leaveTangent=1.1179, tangentMode=RCTM_User, time=0.2042, value=0.6228}
          {arriveTangent=0.0224, interpMode=RCIM_Cubic, leaveTangent=0.0224, tangentMode=RCTM_User, time=1, value=1}
         ]
        }}}
      Scale Curve=0.9
     )))
   /SetVariables(
    Particles.ColorAlt:LinearColor=&MultiplyLinearColor(Linear Color=$Particles.Color, Scale Factor=<1 0.35 0.1 0.02>)
    Particles.SpriteSizeAlt:Vector2f=&Vector2DFromFloat(Value=&FloatFromCurve(
      FloatCurve=DataInterface() {propertyValues {Curve {keys [{arriveTangent=-0.0036, interpMode=RCIM_Cubic, leaveTangent=-0.0036, tangentMode=RCTM_User, time=0, value=1}]}}}
      Scale Curve=&RandomRangeFloat(Minimum=3, Maximum=4.5)
     ))
   ) {isSetParameters=true, module=SetVariables_#2}
   &CameraOffset(Camera Offset Amount=5)
  ]
  rendererClasses [&NiagaraSpriteRendererProperties]
  renderers [
   {rendererClass=&NiagaraSpriteRendererProperties}
   {rendererClass=&NiagaraSpriteRendererProperties, rendererIndex=1}
  ]
 }
 Glow002 {
  emitterUpdateScript [
   &EmitterState()
   &SpawnRate(SpawnRate=15, Use Spawn Probability=false)
  ]
  particleSpawnScript [
   &FN_SetSpawnGroup(Normalize Spawn Group=true, Number of Spawn Groups=11, Particles Per Group=10, Set Ribbon ID and Ribbon Link Order=true)
   /Niagara/Modules/Spawn/Initialization/V2/InitializeParticle(
    Color=<rgba 0 3.3333 10 1>
    Lifetime=&Multiply_Float(A=&RandomRangeFloat(Minimum=0.2, Maximum=0.5), B=1)
    Sprite Size Mode=::NewEnumerator3
    Uniform Sprite Size=&Multiply_Float(A=&RandomRangeFloat(Minimum=0.65, Maximum=1), B=5.5)
    UsePositionOffset=false
    Write Material Random=true
   ) {module=InitializeParticle001}
   /Niagara/Modules/Spawn/Initialization/InitializeParticle(
    Color=<rgba 15 10 1 1>
    Lifetime=&Multiply_Float(A=&UniformRangedFloat(Minimum=0.2, Maximum=0.5), B=1)
    Module.Write Color=true
    Module.Write Lifetime=true
    Module.Write Mass=true
    Module.Write Material Random=true
    Module.Write Position=true
    Module.Write Scale=false
    Module.Write SpriteRotation=false
    Module.Write SpriteSize=true
    Position=&SimulationPosition()
    Sprite Size=&Vector2DFromFloat(Value=&Multiply_Float(A=&UniformRangedFloat(Minimum=0.65, Maximum=1), B=3))
   ) {enabled=false}
   &self:SampleStaticMeshSockets(
    SM=DataInterface() {propertyValues {
      DefaultMesh=null
      InstanceIndex=-1
      LODIndex=0
      PreviewMesh=&FN_Birthday_TieredCake_6
      SoftSourceActor=null
      SourceMode=Default
      bAllowSamplingFromStreamingLODs=false
      bCaptureTransformsPerFrame=true
      bUsePhysicsBodyVelocity=false
      FilteredSockets [Candle_FXSocket_Top]
      LODIndexUserParameter {parameter {name=None, typeDefHandle {registeredTypeIndex=108}, varData []}}
      MeshParameterBinding {parameter {name=None, typeDefHandle {registeredTypeIndex=1}, varData []}}
      SectionFilter {allowedMaterialSlots []}
     }}
    Socket:NiagaraInt32=$Particles.SpawnGroup
   ) {module=SampleStaticMeshSockets001}
   &Color(Color=&MultiplyLinearColor(
     Linear Color=$Particles.Color
     Scale Factor=&UniformRangedVector4(
      Minimum:Vector4f=<1 0.25 0.25 0.01>
      Maximum:Vector4f=<1 1 1 1>
      Randomness Mode:ENiagaraRandomnessMode=::NewEnumerator0
      Random Seed:NiagaraInt32=0
      Fixed Random Seed:NiagaraBool=false
     )
    ))
  ]
  particleUpdateScript [
   &ParticleState()
   &AddVelocity(Coordinate Space=::World, Velocity=&RandomRangeVector(Minimum=<-0.25 -0.25 2>, Maximum=<0.25 0.25 10>))
   &Drag(Drag=25, Use Linear Drag=true, Use Rotational Drag=false)
   &SolveForcesAndVelocity(Clamp Velocity=false, Limit Acceleration=false)
   &ScaleColor(
    Scale Alpha=&FloatFromCurve(FloatCurve=DataInterface() {propertyValues {Curve {keys [<at 0.5 1>,<at 1 0>]}}})
    Scale RGB=&Lerp_Vector(A=<1 1 1>, B=<1 0 0>, Alpha=$Particles.NormalizedAge)
    ScaleA=true
    ScaleRGB=true
   )
   &ScaleSpriteSize(Scale Factor=&Vector2DFromFloat(Value=&FloatFromCurve(
      FloatCurve=DataInterface() {propertyValues {Curve {
         defaultValue=0
         keys [
          {arriveTangent=4.1785, interpMode=RCIM_Cubic, leaveTangent=4.1785, tangentMode=RCTM_User, time=0, value=0}
          {arriveTangent=1.1179, interpMode=RCIM_Cubic, leaveTangent=1.1179, tangentMode=RCTM_User, time=0.2042, value=0.6228}
          {arriveTangent=0.0224, interpMode=RCIM_Cubic, leaveTangent=0.0224, tangentMode=RCTM_User, time=1, value=1}
         ]
        }}}
      Scale Curve=0.9
     )))
   /SetVariables(
    Particles.ColorAlt:LinearColor=&MultiplyLinearColor(Linear Color=$Particles.Color, Scale Factor=<1 0.35 0.1 0.02>)
    Particles.SpriteSizeAlt:Vector2f=&Vector2DFromFloat(Value=&FloatFromCurve(
      FloatCurve=DataInterface() {propertyValues {Curve {keys [{arriveTangent=-0.0036, interpMode=RCIM_Cubic, leaveTangent=-0.0036, tangentMode=RCTM_User, time=0, value=1}]}}}
      Scale Curve=&RandomRangeFloat(Minimum=3, Maximum=4.5)
     ))
   ) {isSetParameters=true, module=SetVariables_#2}
   &CameraOffset(Camera Offset Amount=5)
  ]
  rendererClasses [&NiagaraSpriteRendererProperties]
  renderers [
   {rendererClass=&NiagaraSpriteRendererProperties}
   {rendererClass=&NiagaraSpriteRendererProperties, rendererIndex=1}
  ]
 }
 Light {
  emitterUpdateScript [
   &EmitterState(
    Inactive Response=::NewEnumerator1
    Life Cycle Mode=::NewEnumerator1
    Loop Behavior=::NewEnumerator1
    Loop Duration Mode=::NewEnumerator1
    UseLoopDelay=false
   )
   &SpawnBurst_Instantaneous(Spawn Count=11, Use Spawn Probability=false, UseLoopCountLimit=false)
  ]
  particleSpawnScript [
   &FN_SetSpawnGroup(Normalize Spawn Group=true, Number of Spawn Groups=11, Particles Per Group=10, Set Ribbon ID and Ribbon Link Order=true)
   /SetVariables(Particles.SpriteSize=&Vector2DFromFloat(Value=5)) {isSetParameters=true, module=SetVariables_#6}
   /SetVariables(Particles.Lifetime=&RandomRangeFloat(Minimum=2, Maximum=3)) {isSetParameters=true, module=SetVariables_#7}
   &self:SampleStaticMeshSockets(
    SM=DataInterface() {propertyValues {
      DefaultMesh=null
      InstanceIndex=-1
      LODIndex=0
      PreviewMesh=&FN_Birthday_Tiered_Cake_5
      SoftSourceActor=null
      SourceMode=Default
      bAllowSamplingFromStreamingLODs=false
      bCaptureTransformsPerFrame=true
      bUsePhysicsBodyVelocity=false
      FilteredSockets [
       Candle_FXSocket_01
       Candle_FXSocket_02
       Candle_FXSocket_03
       Candle_FXSocket_04
       Candle_FXSocket_05
       Candle_FXSocket_06
       Candle_FXSocket_07
       Candle_FXSocket_08
       Candle_FXSocket_09
       Candle_FXSocket_10
       Candle_FXSocket_11
      ]
      LODIndexUserParameter {parameter {name=None, typeDefHandle {registeredTypeIndex=108}, varData []}}
      MeshParameterBinding {parameter {name=None, typeDefHandle {registeredTypeIndex=1}, varData []}}
      SectionFilter {allowedMaterialSlots []}
     }}
    Socket:NiagaraInt32=$Particles.SpawnGroup
   ) {module=SampleStaticMeshSockets}
  ]
  particleUpdateScript [
   &ParticleState(Kill Particles When Lifetime Has Elapsed=false, Loop Particles Lifetime=true)
   &FN_Light_v3(
    Color=&RandomRangeLinearColor(Minimum=<rgba 7 3 0 1>, Maximum=<rgba 7 2.5 0 1>)
    Create Color Attribute=true
    Create Exponent Attribute=true
    Create Radius Attribute=true
    Enable Particle Light Enable=true
    Exponent=1
    Light Enabled=true
    Radius:NiagaraFloat=&Multiply_Float(A=&RandomRangeFloat(Minimum=18, Maximum=20), B=0.5)
    Scale Color:NiagaraFloat=&FloatFromCurve(FloatCurve=DataInterface() {propertyValues {Curve {keys [
         <at 0 1 RCIM_Cubic>
         {arriveTangent=1.4927, interpMode=RCIM_Cubic, leaveTangent=1.4927, time=0.0746, value=0.875}
         {arriveTangent=-0.0227, interpMode=RCIM_Cubic, leaveTangent=-0.0227, tangentMode=RCTM_User, time=0.1502, value=1.2241}
         {arriveTangent=-2.2781, interpMode=RCIM_Cubic, leaveTangent=-2.2781, time=0.2439, value=0.3986}
         {arriveTangent=4.6068, interpMode=RCIM_Cubic, leaveTangent=4.6068, time=0.2748, value=0.9402}
         {arriveTangent=2.2278, interpMode=RCIM_Cubic, leaveTangent=2.2278, time=0.3464, value=0.8705}
         {arriveTangent=-2.7898, interpMode=RCIM_Cubic, leaveTangent=-2.7898, time=0.3894, value=1.1955}
         {arriveTangent=-2.6422, interpMode=RCIM_Cubic, leaveTangent=-2.6422, time=0.4785, value=0.5018}
         {arriveTangent=-0.8529, interpMode=RCIM_Cubic, leaveTangent=-0.8529, time=0.4871, value=0.9373}
         {arriveTangent=0.3027, interpMode=RCIM_Cubic, leaveTangent=0.3027, time=0.5752, value=0.4194}
         {arriveTangent=4.1733, interpMode=RCIM_Cubic, leaveTangent=4.1733, time=0.6112, value=0.9748}
         {arriveTangent=2.91, interpMode=RCIM_Cubic, leaveTangent=2.91, time=0.69, value=0.8986}
         {arriveTangent=-2.2524, interpMode=RCIM_Cubic, leaveTangent=-2.2524, time=0.7356, value=1.3367}
         {arriveTangent=-3.2665, interpMode=RCIM_Cubic, leaveTangent=-3.2665, time=0.8416, value=0.5572}
         {arriveTangent=-0.8671, interpMode=RCIM_Cubic, leaveTangent=-0.8671, time=0.8675, value=0.9059}
         {arriveTangent=-0.3671, interpMode=RCIM_Cubic, leaveTangent=-0.3671, time=0.9558, value=0.4582}
         {arriveTangent=-0.6714, interpMode=RCIM_Cubic, leaveTangent=-0.6714, time=0.9899, value=0.8609}
         {arriveTangent=0.7195, interpMode=RCIM_Cubic, leaveTangent=0.7195, time=1.0792, value=0.3753}
         {arriveTangent=1.5649, interpMode=RCIM_Cubic, leaveTangent=1.5649, time=1.099, value=0.9394}
         {arriveTangent=0.3682, interpMode=RCIM_Cubic, leaveTangent=0.3682, time=1.1717, value=0.52}
         {arriveTangent=6.6831, interpMode=RCIM_Cubic, leaveTangent=6.6831, time=1.2078, value=0.9795}
         {arriveTangent=2.1336, interpMode=RCIM_Cubic, leaveTangent=2.1336, time=1.2281, value=0.8972}
         {arriveTangent=1.9914, interpMode=RCIM_Cubic, leaveTangent=1.9914, time=1.311, value=1.1996}
         {arriveTangent=0.4649, interpMode=RCIM_Cubic, leaveTangent=0.4649, time=1.3539, value=1.1478}
         {arriveTangent=-4.574, interpMode=RCIM_Cubic, leaveTangent=-4.574, time=1.396, value=1.2391}
         {arriveTangent=-0.0928, interpMode=RCIM_Cubic, leaveTangent=-0.0928, tangentMode=RCTM_User, time=1.4921, value=0.5155}
         {arriveTangent=5.2872, interpMode=RCIM_Cubic, leaveTangent=5.2872, time=1.5242, value=0.9899}
         {arriveTangent=1.3233, interpMode=RCIM_Cubic, leaveTangent=1.3233, time=1.5566, value=0.8565}
         {arriveTangent=0.3667, interpMode=RCIM_Cubic, leaveTangent=0.3667, time=1.7064, value=1.2309}
         {arriveTangent=-6.7179, interpMode=RCIM_Cubic, leaveTangent=-6.7179, time=1.7959, value=0.9442}
         {arriveTangent=-5.6448, interpMode=RCIM_Cubic, leaveTangent=-5.6448, time=1.8039, value=0.5757}
         {arriveTangent=2.6907, interpMode=RCIM_Cubic, leaveTangent=2.6907, time=1.878, value=0.4804}
         {arriveTangent=3.6859, interpMode=RCIM_Cubic, leaveTangent=3.6859, time=1.9343, value=0.9265}
         {arriveTangent=2.6815, interpMode=RCIM_Cubic, leaveTangent=2.6815, time=1.9912, value=0.8977}
         {arriveTangent=-3.7116, interpMode=RCIM_Cubic, leaveTangent=-3.7116, time=2.0308, value=1.1852}
         {arriveTangent=-1.3071, interpMode=RCIM_Cubic, leaveTangent=-1.3071, time=2.1248, value=0.402}
         {arriveTangent=0.4353, interpMode=RCIM_Cubic, leaveTangent=0.4353, time=2.2158, value=0.9434}
         {arriveTangent=-0.3447, interpMode=RCIM_Cubic, leaveTangent=-0.3447, time=2.3094, value=0.4823}
         {arriveTangent=0.6603, interpMode=RCIM_Cubic, leaveTangent=0.6603, time=2.3258, value=0.9055}
         {arriveTangent=0.3801, interpMode=RCIM_Cubic, leaveTangent=0.3801, time=2.4391, value=0.568}
         {arriveTangent=4.4872, interpMode=RCIM_Cubic, leaveTangent=4.4872, time=2.4654, value=0.9586}
         {arriveTangent=3.4107, interpMode=RCIM_Cubic, leaveTangent=3.4107, time=2.5143, value=0.9055}
         {arriveTangent=-0.515, interpMode=RCIM_Cubic, leaveTangent=-0.515, time=2.5385, value=1.2081}
         {arriveTangent=-2.6117, interpMode=RCIM_Cubic, leaveTangent=-2.6117, time=2.597, value=0.8629}
         {arriveTangent=-2.5605, interpMode=RCIM_Cubic, leaveTangent=-2.5605, time=2.6456, value=0.9286}
         {arriveTangent=0.1009, interpMode=RCIM_Cubic, leaveTangent=0.1009, time=2.7341, value=0.5119}
         {arriveTangent=0.0929, interpMode=RCIM_Cubic, leaveTangent=0.0929, time=2.7509, value=0.9392}
         {arriveTangent=0.2939, interpMode=RCIM_Cubic, leaveTangent=0.2939, time=2.9074, value=0.528}
         {arriveTangent=8.4131, interpMode=RCIM_Cubic, leaveTangent=8.4131, time=2.9271, value=0.991}
         {arriveTangent=0.1236, interpMode=RCIM_Cubic, leaveTangent=0.1236, time=2.957, value=0.9449}
         <at 3 1 RCIM_Cubic>
        ]}}})
   )
  ]
  rendererClasses [&NiagaraLightRendererProperties]
  renderers [{rendererClass=&NiagaraLightRendererProperties, rendererIndex=0}]
 }
 Light001 {
  emitterUpdateScript [
   &EmitterState(
    Inactive Response=::NewEnumerator1
    Life Cycle Mode=::NewEnumerator1
    Loop Behavior=::NewEnumerator1
    Loop Duration Mode=::NewEnumerator1
    UseLoopDelay=false
   )
   &SpawnBurst_Instantaneous(Spawn Count=11, Use Spawn Probability=false, UseLoopCountLimit=false)
  ]
  particleSpawnScript [
   &FN_SetSpawnGroup(Normalize Spawn Group=true, Number of Spawn Groups=11, Particles Per Group=10, Set Ribbon ID and Ribbon Link Order=true)
   /SetVariables(Particles.SpriteSize=&Vector2DFromFloat(Value=5)) {isSetParameters=true, module=SetVariables_#6}
   /SetVariables(Particles.Lifetime=&RandomRangeFloat(Minimum=2, Maximum=3)) {isSetParameters=true, module=SetVariables_#7}
   &self:SampleStaticMeshSockets(
    SM=DataInterface() {propertyValues {
      DefaultMesh=null
      InstanceIndex=-1
      LODIndex=0
      PreviewMesh=&FN_Birthday_TieredCake_6
      SoftSourceActor=null
      SourceMode=Default
      bAllowSamplingFromStreamingLODs=false
      bCaptureTransformsPerFrame=true
      bUsePhysicsBodyVelocity=false
      FilteredSockets [Candle_FXSocket_Top]
      LODIndexUserParameter {parameter {name=None, typeDefHandle {registeredTypeIndex=108}, varData []}}
      MeshParameterBinding {parameter {name=None, typeDefHandle {registeredTypeIndex=1}, varData []}}
      SectionFilter {allowedMaterialSlots []}
     }}
    Socket:NiagaraInt32=$Particles.SpawnGroup
   ) {module=SampleStaticMeshSockets}
  ]
  particleUpdateScript [
   &ParticleState(Kill Particles When Lifetime Has Elapsed=false, Loop Particles Lifetime=true)
   &FN_Light_v3(
    Color=&RandomRangeLinearColor(Minimum=<rgba 7 3 0 1>, Maximum=<rgba 7 2.5 0 1>)
    Create Color Attribute=true
    Create Exponent Attribute=true
    Create Radius Attribute=true
    Enable Particle Light Enable=true
    Exponent=1
    Light Enabled=true
    Radius:NiagaraFloat=&Multiply_Float(A=&RandomRangeFloat(Minimum=18, Maximum=20), B=0.5)
    Scale Color:NiagaraFloat=&FloatFromCurve(FloatCurve=DataInterface() {propertyValues {Curve {keys [
         <at 0 1 RCIM_Cubic>
         {arriveTangent=1.4927, interpMode=RCIM_Cubic, leaveTangent=1.4927, time=0.0746, value=0.875}
         {arriveTangent=-0.0227, interpMode=RCIM_Cubic, leaveTangent=-0.0227, tangentMode=RCTM_User, time=0.1502, value=1.2241}
         {arriveTangent=-2.2781, interpMode=RCIM_Cubic, leaveTangent=-2.2781, time=0.2439, value=0.3986}
         {arriveTangent=4.6068, interpMode=RCIM_Cubic, leaveTangent=4.6068, time=0.2748, value=0.9402}
         {arriveTangent=2.2278, interpMode=RCIM_Cubic, leaveTangent=2.2278, time=0.3464, value=0.8705}
         {arriveTangent=-2.7898, interpMode=RCIM_Cubic, leaveTangent=-2.7898, time=0.3894, value=1.1955}
         {arriveTangent=-2.6422, interpMode=RCIM_Cubic, leaveTangent=-2.6422, time=0.4785, value=0.5018}
         {arriveTangent=-0.8529, interpMode=RCIM_Cubic, leaveTangent=-0.8529, time=0.4871, value=0.9373}
         {arriveTangent=0.3027, interpMode=RCIM_Cubic, leaveTangent=0.3027, time=0.5752, value=0.4194}
         {arriveTangent=4.1733, interpMode=RCIM_Cubic, leaveTangent=4.1733, time=0.6112, value=0.9748}
         {arriveTangent=2.91, interpMode=RCIM_Cubic, leaveTangent=2.91, time=0.69, value=0.8986}
         {arriveTangent=-2.2524, interpMode=RCIM_Cubic, leaveTangent=-2.2524, time=0.7356, value=1.3367}
         {arriveTangent=-3.2665, interpMode=RCIM_Cubic, leaveTangent=-3.2665, time=0.8416, value=0.5572}
         {arriveTangent=-0.8671, interpMode=RCIM_Cubic, leaveTangent=-0.8671, time=0.8675, value=0.9059}
         {arriveTangent=-0.3671, interpMode=RCIM_Cubic, leaveTangent=-0.3671, time=0.9558, value=0.4582}
         {arriveTangent=-0.6714, interpMode=RCIM_Cubic, leaveTangent=-0.6714, time=0.9899, value=0.8609}
         {arriveTangent=0.7195, interpMode=RCIM_Cubic, leaveTangent=0.7195, time=1.0792, value=0.3753}
         {arriveTangent=1.5649, interpMode=RCIM_Cubic, leaveTangent=1.5649, time=1.099, value=0.9394}
         {arriveTangent=0.3682, interpMode=RCIM_Cubic, leaveTangent=0.3682, time=1.1717, value=0.52}
         {arriveTangent=6.6831, interpMode=RCIM_Cubic, leaveTangent=6.6831, time=1.2078, value=0.9795}
         {arriveTangent=2.1336, interpMode=RCIM_Cubic, leaveTangent=2.1336, time=1.2281, value=0.8972}
         {arriveTangent=1.9914, interpMode=RCIM_Cubic, leaveTangent=1.9914, time=1.311, value=1.1996}
         {arriveTangent=0.4649, interpMode=RCIM_Cubic, leaveTangent=0.4649, time=1.3539, value=1.1478}
         {arriveTangent=-4.574, interpMode=RCIM_Cubic, leaveTangent=-4.574, time=1.396, value=1.2391}
         {arriveTangent=-0.0928, interpMode=RCIM_Cubic, leaveTangent=-0.0928, tangentMode=RCTM_User, time=1.4921, value=0.5155}
         {arriveTangent=5.2872, interpMode=RCIM_Cubic, leaveTangent=5.2872, time=1.5242, value=0.9899}
         {arriveTangent=1.3233, interpMode=RCIM_Cubic, leaveTangent=1.3233, time=1.5566, value=0.8565}
         {arriveTangent=0.3667, interpMode=RCIM_Cubic, leaveTangent=0.3667, time=1.7064, value=1.2309}
         {arriveTangent=-6.7179, interpMode=RCIM_Cubic, leaveTangent=-6.7179, time=1.7959, value=0.9442}
         {arriveTangent=-5.6448, interpMode=RCIM_Cubic, leaveTangent=-5.6448, time=1.8039, value=0.5757}
         {arriveTangent=2.6907, interpMode=RCIM_Cubic, leaveTangent=2.6907, time=1.878, value=0.4804}
         {arriveTangent=3.6859, interpMode=RCIM_Cubic, leaveTangent=3.6859, time=1.9343, value=0.9265}
         {arriveTangent=2.6815, interpMode=RCIM_Cubic, leaveTangent=2.6815, time=1.9912, value=0.8977}
         {arriveTangent=-3.7116, interpMode=RCIM_Cubic, leaveTangent=-3.7116, time=2.0308, value=1.1852}
         {arriveTangent=-1.3071, interpMode=RCIM_Cubic, leaveTangent=-1.3071, time=2.1248, value=0.402}
         {arriveTangent=0.4353, interpMode=RCIM_Cubic, leaveTangent=0.4353, time=2.2158, value=0.9434}
         {arriveTangent=-0.3447, interpMode=RCIM_Cubic, leaveTangent=-0.3447, time=2.3094, value=0.4823}
         {arriveTangent=0.6603, interpMode=RCIM_Cubic, leaveTangent=0.6603, time=2.3258, value=0.9055}
         {arriveTangent=0.3801, interpMode=RCIM_Cubic, leaveTangent=0.3801, time=2.4391, value=0.568}
         {arriveTangent=4.4872, interpMode=RCIM_Cubic, leaveTangent=4.4872, time=2.4654, value=0.9586}
         {arriveTangent=3.4107, interpMode=RCIM_Cubic, leaveTangent=3.4107, time=2.5143, value=0.9055}
         {arriveTangent=-0.515, interpMode=RCIM_Cubic, leaveTangent=-0.515, time=2.5385, value=1.2081}
         {arriveTangent=-2.6117, interpMode=RCIM_Cubic, leaveTangent=-2.6117, time=2.597, value=0.8629}
         {arriveTangent=-2.5605, interpMode=RCIM_Cubic, leaveTangent=-2.5605, time=2.6456, value=0.9286}
         {arriveTangent=0.1009, interpMode=RCIM_Cubic, leaveTangent=0.1009, time=2.7341, value=0.5119}
         {arriveTangent=0.0929, interpMode=RCIM_Cubic, leaveTangent=0.0929, time=2.7509, value=0.9392}
         {arriveTangent=0.2939, interpMode=RCIM_Cubic, leaveTangent=0.2939, time=2.9074, value=0.528}
         {arriveTangent=8.4131, interpMode=RCIM_Cubic, leaveTangent=8.4131, time=2.9271, value=0.991}
         {arriveTangent=0.1236, interpMode=RCIM_Cubic, leaveTangent=0.1236, time=2.957, value=0.9449}
         <at 3 1 RCIM_Cubic>
        ]}}})
   )
  ]
  rendererClasses [&NiagaraLightRendererProperties]
  renderers [{rendererClass=&NiagaraLightRendererProperties, rendererIndex=0}]
 }
 Smoke {
  emitterUpdateScript [
   &EmitterState()
   &SpawnRate(SpawnRate=25, Use Spawn Probability=false)
  ]
  particleSpawnScript [
   &FN_SetSpawnGroup(Normalize Spawn Group=true, Number of Spawn Groups=11, Particles Per Group=10, Set Ribbon ID and Ribbon Link Order=true)
   &InitializeRibbon(
    Color=<rgba 2 2 2 0.5>
    Lifetime=&Multiply_Float(
     A=&RandomRangeFloat(Minimum=1.25, Maximum=1.75, Randomness Mode=::NewEnumerator1, Random Seed=$Particles.SpawnGroup)
     B=1.5
    )
    Module.Write Color=true
    Module.Write Lifetime=true
    Module.Write Mass=true
    Module.Write Material Random=true
    Module.Write Position=true
    Module.Write RibbonTwist=false
    Module.Write RibbonWidth=true
    Ribbon Width=&Multiply_Float(A=&RandomRangeFloat(Minimum=2.5, Maximum=3.5), B=2)
   )
   &self:SampleStaticMeshSockets(
    SM=DataInterface() {propertyValues {
      DefaultMesh=null
      InstanceIndex=-1
      LODIndex=0
      PreviewMesh=&FN_Birthday_TieredCake_6
      SoftSourceActor=null
      SourceMode=Default
      bAllowSamplingFromStreamingLODs=false
      bCaptureTransformsPerFrame=true
      bUsePhysicsBodyVelocity=false
      FilteredSockets [
       Candle_FXSocket_01
       Candle_FXSocket_02
       Candle_FXSocket_03
       Candle_FXSocket_04
       Candle_FXSocket_05
       Candle_FXSocket_06
       Candle_FXSocket_07
       Candle_FXSocket_08
       Candle_FXSocket_09
       Candle_FXSocket_10
       Candle_FXSocket_11
      ]
      LODIndexUserParameter {parameter {name=None, typeDefHandle {registeredTypeIndex=108}, varData []}}
      MeshParameterBinding {parameter {name=None, typeDefHandle {registeredTypeIndex=1}, varData []}}
      SectionFilter {allowedMaterialSlots []}
     }}
    Socket:NiagaraInt32=$Particles.SpawnGroup
   ) {module=SampleStaticMeshSockets001}
  ]
  particleUpdateScript [
   &ParticleState()
   &GravityForce(Gravity=<0 0 75>)
   /Niagara/Modules/Update/Forces/CurlNoiseForce(
    Noise Frequency=0.01
    Noise Strength=&FloatFromCurve(
     FloatCurve=DataInterface() {propertyValues {Curve {keys [
         {arriveTangent=1.0245, interpMode=RCIM_Cubic, leaveTangent=1.0245, tangentMode=RCTM_User, time=0, value=0}
         {arriveTangent=1.6798, interpMode=RCIM_Cubic, leaveTangent=1.6798, tangentMode=RCTM_User, time=1, value=1}
        ]}}}
     Scale Curve=30
    )
    Pan Noise=false
    Random Seed:NiagaraInt32=&RandomRangeInt(Minimum=0, Maximum=100)
   )
   &Drag(Drag=2.5, Use Linear Drag=true, Use Rotational Drag=false) {module=Drag001}
   &FN_RibbonMaterial(Color Pan U Speed=-0.5, Erosion Amount=1, Macro Distortion Amount=1, Opacity Erosion=0.25, Opacity Pan U Speed=-0.5, Opacity Power=0.35)
   &ScaleRibbonWidth(Ribbon Width Scale=&FloatFromCurve(
     FloatCurve=DataInterface() {propertyValues {Curve {keys [
         {arriveTangent=0.0219, interpMode=RCIM_Cubic, leaveTangent=0.0219, tangentMode=RCTM_User, time=0, value=0}
         <at 0.5 0.9666 RCIM_Cubic>
         {arriveTangent=-0.2116, interpMode=RCIM_Cubic, leaveTangent=-0.2116, tangentMode=RCTM_User, time=1, value=0}
        ]}}}
     Scale Curve=&RandomRangeFloat(Minimum=1, Maximum=1)
    )) {module=ScaleRibbonWidth001}
   &SolveForcesAndVelocity(Clamp Velocity=false, Limit Acceleration=false)
   &ScaleColor(
    Scale Alpha=&FloatFromCurve(
     FloatCurve=DataInterface() {propertyValues {Curve {keys [
         {arriveTangent=9.0725, interpMode=RCIM_Cubic, leaveTangent=9.0725, tangentMode=RCTM_User, time=0, value=0}
         {arriveTangent=-0.2513, interpMode=RCIM_Cubic, leaveTangent=-0.2513, tangentMode=RCTM_User, time=0.1408, value=0.9847}
         {arriveTangent=-0.0405, interpMode=RCIM_Cubic, leaveTangent=-0.0405, tangentMode=RCTM_User, time=1, value=0}
        ]}}}
     Scale Curve=0.1
    )
    ScaleA=true
    ScaleRGB=true
   )
  ]
  rendererClasses [&NiagaraRibbonRendererProperties]
  renderers [{rendererClass=&NiagaraRibbonRendererProperties}]
 }
 Smoke001 {
  emitterUpdateScript [
   &EmitterState()
   &SpawnRate(SpawnRate=25, Use Spawn Probability=false)
  ]
  particleSpawnScript [
   &FN_SetSpawnGroup(Normalize Spawn Group=true, Number of Spawn Groups=11, Particles Per Group=10, Set Ribbon ID and Ribbon Link Order=true)
   &InitializeRibbon(
    Color=<rgba 1 1 1 0.1>
    Lifetime=&Multiply_Float(
     A=&RandomRangeFloat(Minimum=1.25, Maximum=1.75, Randomness Mode=::NewEnumerator1, Random Seed=$Particles.SpawnGroup)
     B=1.5
    )
    Module.Write Color=true
    Module.Write Lifetime=true
    Module.Write Mass=true
    Module.Write Material Random=true
    Module.Write Position=true
    Module.Write RibbonTwist=false
    Module.Write RibbonWidth=true
    Ribbon Width=&Multiply_Float(A=&RandomRangeFloat(Minimum=2.5, Maximum=3.5), B=2)
   )
   &self:SampleStaticMeshSockets(
    SM=DataInterface() {propertyValues {
      DefaultMesh=null
      InstanceIndex=-1
      LODIndex=0
      PreviewMesh=&FN_Birthday_TieredCake_6
      SoftSourceActor=null
      SourceMode=Default
      bAllowSamplingFromStreamingLODs=false
      bCaptureTransformsPerFrame=true
      bUsePhysicsBodyVelocity=false
      FilteredSockets [Candle_FXSocket_Top]
      LODIndexUserParameter {parameter {name=None, typeDefHandle {registeredTypeIndex=108}, varData []}}
      MeshParameterBinding {parameter {name=None, typeDefHandle {registeredTypeIndex=1}, varData []}}
      SectionFilter {allowedMaterialSlots []}
     }}
    Socket:NiagaraInt32=$Particles.SpawnGroup
   ) {module=SampleStaticMeshSockets001}
  ]
  particleUpdateScript [
   &ParticleState()
   &GravityForce(Gravity=<0 0 75>)
   &CurlNoiseForce(
    Noise Frequency=0.01
    Noise Strength=&FloatFromCurve(
     FloatCurve=DataInterface() {propertyValues {Curve {keys [
         {arriveTangent=1.0245, interpMode=RCIM_Cubic, leaveTangent=1.0245, tangentMode=RCTM_User, time=0, value=0}
         {arriveTangent=1.6798, interpMode=RCIM_Cubic, leaveTangent=1.6798, tangentMode=RCTM_User, time=1, value=1}
        ]}}}
     Scale Curve=50
    )
    Pan Noise=false
    Random Seed:NiagaraInt32=&RandomRangeInt(Minimum=0, Maximum=100)
   )
   &Drag(Drag=2.5, Use Linear Drag=true, Use Rotational Drag=false) {module=Drag001}
   &FN_RibbonMaterial(Color Pan U Speed=-0.5, Erosion Amount=1, Macro Distortion Amount=1, Opacity Erosion=0.25, Opacity Pan U Speed=-0.5, Opacity Power=0.35)
   &ScaleRibbonWidth(Ribbon Width Scale=&FloatFromCurve(
     FloatCurve=DataInterface() {propertyValues {Curve {keys [
         {arriveTangent=0.0219, interpMode=RCIM_Cubic, leaveTangent=0.0219, tangentMode=RCTM_User, time=0, value=0}
         <at 0.5 0.9666 RCIM_Cubic>
         {arriveTangent=-0.2116, interpMode=RCIM_Cubic, leaveTangent=-0.2116, tangentMode=RCTM_User, time=1, value=0}
        ]}}}
     Scale Curve=&RandomRangeFloat(Minimum=1, Maximum=1)
    )) {module=ScaleRibbonWidth001}
   &SolveForcesAndVelocity(Clamp Velocity=false, Limit Acceleration=false)
   &ScaleColor(
    Scale Alpha=&FloatFromCurve(
     FloatCurve=DataInterface() {propertyValues {Curve {keys [
         {arriveTangent=9.0725, interpMode=RCIM_Cubic, leaveTangent=9.0725, tangentMode=RCTM_User, time=0, value=0}
         {arriveTangent=0.0486, interpMode=RCIM_Cubic, leaveTangent=0.0486, tangentMode=RCTM_User, time=0.1408, value=0.9847}
         {arriveTangent=-0.0405, interpMode=RCIM_Cubic, leaveTangent=-0.0405, tangentMode=RCTM_User, time=1, value=0}
        ]}}}
     Scale Curve=0.1
    )
    ScaleA=true
    ScaleRGB=true
   )
  ]
  rendererClasses [&NiagaraRibbonRendererProperties]
  renderers [{rendererClass=&NiagaraRibbonRendererProperties}]
 }
}
summary {
 systemName=NS_BirthdayCake_8_CandleFlame
 emitters [
  {emitterName=Glow, rendererClasses [&NiagaraSpriteRendererProperties]}
  {emitterName=Light, rendererClasses [&NiagaraLightRendererProperties]}
  {emitterName=Smoke, rendererClasses [&NiagaraRibbonRendererProperties]}
  {emitterName=Glow002, rendererClasses [&NiagaraSpriteRendererProperties]}
  {emitterName=Smoke001, rendererClasses [&NiagaraRibbonRendererProperties]}
  {emitterName=Light001, rendererClasses [&NiagaraLightRendererProperties]}
  {emitterName=FlamesAndGlows, rendererClasses [&NiagaraSpriteRendererProperties, &NiagaraMeshRendererProperties]}
  {
   emitterName=FlamesAndGlows001
   rendererClasses [&NiagaraSpriteRendererProperties, &NiagaraMeshRendererProperties]
  }
 ]
 userVariables [
  User.bIsFrontEnd {type=NiagaraBool}
  User.bIsFrontEndPreview {type=NiagaraBool}
 ]
}
